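# M_OctopusDripMask is an unlit translucent material for Niagara mesh particles, drawing a foamy drip mask around a water plane height.
# Vertical distance from the WaterPlaneLocation parameter, scaled by MaskHeight, forms a falloff broken up by Voronoi noise.
# A second sample of that texture distorts the noise UVs, with parameters offsetting both vertically.
# The mask indexes foam and biome curve-atlas tints, whose color and alpha blend by a biome mask, with color passing through emissive adjustments.
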Material /BallerV2/FX/M_OctopusDripMask
graph {
 properties {
  bUsedWithNiagaraMeshParticles=true
  blendMode=BLEND_Translucent
  pixelDepthOffsetMode=PDOM_Legacy
  shadingModel=MSM_Unlit
  usedShadingModels=Unlit
 }
 nodes [
  n1 = Subtract(A=WorldPosition().XYZ, B=VectorParameter() {parameterName=WaterPlaneLocation, defaultValue=<rgba 0 0 0 1>}.RGB)
  n2 = ScalarParameter() {defaultValue=32, parameterName=MaskHeight}
  n3 = Power(Base=Abs(None=Divide(A=@n1, B=@n2) {constB=32})) {constExponent=1.5}
  n4 = TextureCoordinate() {uTiling=2, vTiling=1.5}
  n5 = TextureCoordinate() {uTiling=2, vTiling=1.5}
  n6 = AppendVector(A=Constant(), B=ScalarParameter() {parameterName=OffsetDistortion})
  n7 = Multiply(A=TextureSample(UVs=Add(A=@n5, B=@n6)) {samplerType=SAMPLERTYPE_LinearGrayscale, texture=/Game/Packages/Fortress_Sky/SkyDome/Master/T_SkyDome_Cloud_Voroni}.R) {constB=0.15}
  n8 = Constant()
  n9 = ScalarParameter() {parameterName=OffsetSpeed}
  n10 = TextureSample(UVs=Add(A=Add(A=@n4, B=@n7), B=AppendVector(A=@n8, B=@n9))) {samplerType=SAMPLERTYPE_LinearGrayscale, texture=&T_SkyDome_Cloud_Voroni}
  n11 = Add(A=@n10.R) {constB=0.5}
  n12 = Saturate(None=OneMinus(None=Multiply(A=@n3, B=@n11)))
  n13 = Multiply(A=Multiply(A=ComponentMask(None=@n12) {b=true}, B=OneMinus(None=@n10.R))) {constB=0.4}
  n14 = CurveAtlasRowParameter(CurveTime=@n13) {
   atlas=/Game/Effects/Niagara/Emitters/Water/WaterAtlas
   curve=/Game/Effects/Niagara/Emitters/Water/FoamEffect/Curve_FoamTint
   defaultValue=5
   parameterName=TintCurve
  }
  n15 = CurveAtlasRowParameter(CurveTime=@n13) {
   atlas=&WaterAtlas
   curve=/Game/Effects/Niagara/Emitters/Water/FoamEffect/Curve_BiomeTint
   defaultValue=20
   parameterName=TintCurve_Biome
  }
  n16 = MaterialFunctionCall() {materialFunction=/Game/Effects/Niagara/Emitters/Water/FoamEffect/MF_NFX_BiomeMask}
  LinearInterpolate(A=@n14.A, B=@n15.A, Alpha=@n16.Result)
  MaterialFunctionCall(Emissive=LinearInterpolate(A=@n14, B=@n15, Alpha=@n16.Result)) {materialFunction=&MF_FX_EmissiveAdjustments}
 ]
}
