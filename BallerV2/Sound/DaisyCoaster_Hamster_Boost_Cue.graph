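# Sound cue for the DaisyCoaster hamster vehicle's boost sound, played in the Fort_Vehicles sound class.
# A Random node picks one of three equally weighted boost waves, then a Modulator varies its pitch upward and its volume for variety.
# Close-movement attenuation spatializes it, and a concurrency override allows one instance at a time, blocking new plays while one is active.
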SoundCue /BallerV2/Sound/DaisyCoaster_Hamster_Boost_Cue
root=Modulator_0
nodes [
 {id=Modulator_0, children [Random_0], properties {pitchMax=1.2, pitchMin=1, volumeMin=0.65}}
 {id=Random_0, children [WavePlayer_0,WavePlayer_1,WavePlayer_6], properties {weights [1,1,1]}}
 {id=WavePlayer_0, properties {soundWaveAssetPtr=&/DaisyCoaster_Hamster_Boost_01}}
 {id=WavePlayer_1, properties {soundWaveAssetPtr=&/DaisyCoaster_Hamster_Boost_02}}
 {id=WavePlayer_6, properties {soundWaveAssetPtr=&/DaisyCoaster_Hamster_Boost_03}}
]
properties {
 attenuationSettings=&/Fort_DaisyCoaster_Movement_Close_Att
 bOverrideConcurrency=true
 maxDistance=5300
 soundClassObject=&Fort_Vehicles
 volumeMultiplier=1.05
 concurrencyOverrides {maxCount=1, resolutionRule=PreventNew}
}
